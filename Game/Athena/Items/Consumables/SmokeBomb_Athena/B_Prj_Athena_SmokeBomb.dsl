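; Smoke bomb projectile for Athena that explodes into a smoke effect on contact.
; A bounce or a capsule overlap ends the fuse, with checks to ignore the instigator's pawn, and the fuse end triggers the explosion.
; Where cosmetic events can run, the explosion spawns the smoke emitter, hides the mesh and deactivates the projectile's particles.
; A failsafe timer started at BeginPlay destroys the actor without triggering an explosion.
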(event EventBeginPlay
  (|CallParentFunction)
  (Utilities|Time|SetTimerbyEvent (delegate Custom|FailsafeTimer) 2.0))

(event Custom|FailsafeTimer
  (Actor|DestroyActor))

(event OnComponentBeginOverlap(CapsuleComponent)
  (params (OverlappedComponent :type Primitive%20Component%20Object%20Reference) (OtherActor :type Actor%20Object%20Reference) (OtherComp :type Primitive%20Component%20Object%20Reference) (OtherBodyIndex :type Integer) (bFromSweep :type Boolean) (SweepResult :type Hit%20Result%20Structure%20%28by%20ref%29))
  (bind (_bblockinghit _binitialoverlap _time _distance _location _impactpoint _normal _impactnormal _physmat _hitactor) (Collision|BreakHitResult SweepResult))
  ;[1 1 1] If we overlap a pawn, fake a bounce.
  (bind _aspawn (Utilities|Casting|CastToPawn _hitactor))
  (if (Utilities|NotEqual(Object) _aspawn (Game|GetInstigator))
    (CallFunction|FuseEnded)))

(event EventOnBounce
  (params (Hit :type Hit%20Result%20Structure%20%28by%20ref%29))
  (bind (_bblockinghit _binitialoverlap _time _distance _location _impactpoint _normal _impactnormal _physmat _hitactor) (Collision|BreakHitResult Hit))
  ;[1 1 1] OnBounce - After X bounces set to explode
  (|CallParentFunction Hit)
  (bind _aspawn (Utilities|Casting|CastToPawn _hitactor)
    (:then
      (if (Utilities|NotEqual(Object) _aspawn (Game|GetInstigator))
        _))
    (:CastFailed
      _))
  (CallFunction|FuseEnded))

(event EventFuseEnded
  ;[1 1 1] Fuse ended - explode
  (Delivery|DoExplosion))

(event EventOnExploded
  (params (HitActors :type Array%20of%20Actor%20Object%20References) (HitResults :type Array%20of%20Hit%20Result%20Structures))
  ;[1 1 1] OnExploded - Emitter, Sound
  (|CallParentFunction HitActors HitResults)
  (Networking|CanExecuteCosmeticEvents
    (:True
      (bind _returnvalue (Effects|SpawnEmitteratLocation (Variables|Default|GetP_SmokeGrenadeExplosion) (Transformation|GetActorLocation)))
      (Rendering|SetTranslucentSortPriority _returnvalue -1)
      (Development|SetHiddeninGame (Variables|Default|GetSkeletalMesh) true false)
      (Components|Activation|Deactivate (Variables|Default|GetParticleSystem)))))

(fn ConstructionScript
  (|CallParentFunction))
