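; Gameplay ability for the Shadow Bomb (SmokeBomb) consumable that shrouds the player for a timed duration.
; Activation checks inventory space, holsters weapons, grants a gadget (variant chosen by backpack parts) and applies the shroud loop effect.
; Expiry while falling waits for landing or movement states like ziplines or vehicles; death or DBNO removes the effect immediately.
; Ending removes effects and gadget, purges zero-count bombs, restores weapons and applies a post-use quest tag effect.
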(event Custom|OnPlayerDiedCleanUp
  (params (DamagedActor :type Actor%20Object%20Reference) (Damage :type Float%20%28single-precision%29) (InstigatedBy :type Controller%20Object%20Reference) (DamageCauser :type Actor%20Object%20Reference) (HitLocation :type Vector) (FHitComponent :type Primitive%20Component%20Object%20Reference) (BoneName :type Name) (Momentum :type Vector))
  (Ability|RemoveGameplayEffectFromOwnerWithHandle (Variables|Default|GetGEShroudGC))
  (Ability|RemoveGameplayEffectFromOwnerWithAssetTags (Variables|Default|GetGEAssetTag)) ; Ends the ability
)

(event Custom|OnPlayerStartDBNO_Event_0
  (Ability|RemoveGameplayEffectFromOwnerWithHandle (Variables|Default|GetGEShroudGC))
  (Ability|RemoveGameplayEffectFromOwnerWithAssetTags (Variables|Default|GetGEAssetTag)) ; Ends the ability
)

(event Ability|EventActivateAbility
  (bind (_owneractor _avataractor) (Utilities|Struct|BreakGameplayAbilityActorInfo (Ability|GetActorInfo)))
  (if (not (Variables|Default|GetApplied))
    (Variables|Default|SetApplied true)
    (bind _output_get (Variables|Default|SetPlayerPawn (Utilities|Casting|CastToFortPlayerPawn _avataractor)))
    ;[0 0.15 0.1389 0.5] To do the GE cleanup if player directly dies (not DBNO) like going out of bounds.
    (EventDispatchers|BindEventtoOnDied _output_get (EventDispatchers|CreateEvent :!event Custom|OnPlayerDiedCleanUp))
    (EventDispatchers|AssignOnPlayerStartDBNO _output_get (delegate Custom|OnPlayerStartDBNO_Event_0))
    (bind _bcanactivateability (CallFunction|CanActivateAbility)) ; Check if we can activate the ability by ensure the pawn has enough space for the inventory item
    (if _bcanactivateability
      (Weapon|SetHolsterWeapon (Variables|Default|GetPlayerPawn) 0 true false (Variables|Default|GetSmokeHolster))
      (CallFunction|GiveItem (Fortnite|GetFortPlayerControllerfromActor (Variables|Default|GetPlayerPawn)))
      (bind _returnvalue (GameplayEffects|ApplyGameplayEffectToSelf (Ability|GetAbilitySystemComponent (Variables|Default|GetPlayerPawn)) "/Game/Athena/Items/Consumables/SmokeBomb_Athena/GE_ShadowBOmbLoop.GE_ShadowBombLoop_C" 0.0 (Utilities|Struct|MakeGameplayEffectContextHandle)))
      (Variables|Default|SetGEShroudGC _returnvalue)
      (Utilities|Time|SetTimerbyEvent (delegate Custom|InAirCheck) (ScalableFloat|GetValueatLevel (Variables|Default|GetSmokeBombDuration))) ; Remove the shroud after a duration unless were in the air
      (else
        (Ability|CancelAbility)))))

(event Custom|InAirCheck
  (if (not (Math|Boolean|ORBoolean (not (AI|Components|NavMovement|IsFalling (Variables|/Script/Engine.Character|GetCharacterMovement (Variables|Default|GetPlayerPawn)))) (Vehicle|IsinVehicle (Variables|Default|GetPlayerPawn)) (SkyTube|IsinSkyTube (Variables|Default|GetPlayerPawn) :!class /Script/FortniteGame.FortPlayerPawn) (Variables|/Script/FortniteGame.FortPawn|GetbIsSliding (Variables|Default|GetPlayerPawn))))
    (EventDispatchers|BindEventtoZiplineStateChanged (Variables|Default|GetPlayerPawn) (delegate Custom|EnteredZiplineAfterExpiration))
    (Ability|Tasks|WaitMovementModeChange "MOVE_Walking"
      (:then
        (Ability|Tasks|WaitGameplayTagAdd "(TagName=\"Gameplay.Status.Skytubing\")"
          (:then
            (Ability|Tasks|WaitGameplayEvent "(TagName=\"Event.Vehicle.Entered\")"
              (:then
                (Ability|Tasks|WaitGameplayTagAdd "(TagName=\"Granted.Riding.Player.IsRiding\")"
                  (:then
                    (Ability|Tasks|WaitGameplayTagAdd "(TagName=\"MovementMode.Sliding\")"
                      (:then
                        (Ability|Tasks|WaitGameplayTagAdd "(TagName=\"Gameplay.Action.Player.Grinding\")"
                          (:Added
                            _)))
                      (:Added
                        _)))
                  (:Added
                    _))
                (Ability|RemoveGameplayEffectFromOwnerWithHandle (Variables|Default|GetGEShroudGC))
                (Ability|RemoveGameplayEffectFromOwnerWithAssetTags (Variables|Default|GetGEAssetTag)) ; Ends the ability
              )
              (:EventReceived
                _)) ; FORT-184515 - Fix for player entering a vehicle after effect expires caused the effect to stay active until they exited vehicle
          )
          (:Added
            _)))
      (:OnChange
        _)) ; Wait til on the ground or ziplining if in the air
  ))

(event Custom|EnteredZiplineAfterExpiration
  (params (bIsZiplining :type Boolean) (FortPlayerPawn :type Fort%20Player%20Pawn%20Object%20Reference))
  (Ability|RemoveGameplayEffectFromOwnerWithHandle (Variables|Default|GetGEShroudGC))
  (Ability|RemoveGameplayEffectFromOwnerWithAssetTags (Variables|Default|GetGEAssetTag)) ; Ends the ability
)

(event Ability|EventOnEndAbility
  (params (bWasCancelled :type Boolean))
  (bind _returnvalue (Ability|GetOwningActorfromActorInfo))
  (bind _returnvalue_2 (Ability|GetAbilitySystemComponent _returnvalue))
  (bind _playerpawn (Variables|Default|GetPlayerPawn))
  (bind _returnvalue_3 (Fortnite|GetFortPlayerControllerfromActor _playerpawn))
  (bind _playerpawn_2 (Variables|Default|GetPlayerPawn))
  ;[1 1 1] On End Ability
  (Ability|SendGameplayEvent (Variables|Default|GetShroudDurationComplete) (Variables|Default|GetEventData))
  (GameplayEffects|RemoveActiveGameplayEffectbySourceEffect _returnvalue_2 "/Game/Athena/Items/Consumables/SmokeBomb_Athena/GE_ApplySmokeBomb.GE_ApplySmokeBomb_C" 0 -1)
  (GameplayEffects|RemoveActiveGameplayEffectbySourceEffect _returnvalue_2 "/Game/Athena/Items/Consumables/SmokeBomb_Athena/GE_ShadowBOmbLoop.GE_ShadowBombLoop_C" 0 -1)
  (Utilities|IsValid _returnvalue_3
    (:Is%20Valid
      (CallFunction|RemoveItem _returnvalue_3)
      ;; Added to account for 0 cnt item
      (CallFunction|CheckWeaponListforShadowBomb)
      (Utilities|IsValid _playerpawn_2
        (:Is%20Valid
          (Weapon|SetHolsterWeapon _playerpawn_2 0 false false (Variables|Default|GetSmokeHolster))
          (Variables|Default|SetApplied false)
          ;;[1 0.9027 0.639] Deal damage to an opponent within 10s after using a Shadow Bomb challenge
          (GameplayEffects|ApplyGameplayEffectToSelf :self (Ability|GetAbilitySystemComponent (Variables|Default|GetPlayerPawn)) :GameplayEffectClass "/Game/Athena/Items/QuestInteractables/Quest_GEs/GE_Athena_Quests_PostShadowBombTag.GE_Athena_Quests_PostShadowBombTag_C" :Level 0.0)
          ;[1 1 1] Unbind from bound event dispatchers
          (Utilities|IsValid (Variables|Default|GetPlayerPawn)
            (:Is%20Valid
              (EventDispatchers|UnbindEventfromOnPlayerStartDBNO (Variables|Default|GetPlayerPawn) (EventDispatchers|CreateEvent :!event Custom|OnPlayerStartDBNO_Event_0))
              (EventDispatchers|UnbindEventfromZiplineStateChanged (Variables|Default|GetPlayerPawn) (EventDispatchers|CreateEvent :!event Custom|EnteredZiplineAfterExpiration))
              (EventDispatchers|UnbindEventfromOnDied (Variables|Default|GetPlayerPawn) (EventDispatchers|CreateEvent :!event Custom|OnPlayerDiedCleanUp)))))))))

(event Custom|RemoveGUIDFromInvWithZeroCount
  (params (Item%20GUID :type Guid%20Structure))
  ;[0.9167 0.9167 0.9167 0.5] Remove item with given GUID from player inventory if count is 0
  (bind _outfortplayerpawn (CallFunction|GetAvatarasValidFortPlayerPawn)
    (:ValidFortPlayerPawn
      (bind _returnvalue (Inventory|GetItemNuminStackbyGuid (Fortnite|GetFortPlayerControllerfromActor _outfortplayerpawn) Item%20GUID))
      (if (== _returnvalue 0)
        (bind _returnvalue_2 (Inventory|GetInventoryItemwithGuid (Fortnite|GetFortPlayerControllerfromActor _outfortplayerpawn) Item%20GUID))
        (Inventory|RemoveFortItemFromPlayer (Fortnite|GetFortPlayerControllerfromActor _outfortplayerpawn) _returnvalue_2 -1 true)))))

(event Custom|CheckWeaponListForShadowBomb
  ;[1 1 1 0.5] Check if duplicate snowman items exist on player and remove unneeded ones if item count 0
  (bind _outfortplayerpawn (CallFunction|GetAvatarasValidFortPlayerPawn)
    (:ValidFortPlayerPawn
      (for _array_element (Variables|/Script/FortniteGame.FortPawn|GetCurrentWeaponList _outfortplayerpawn)
        (if (== (Utilities|GetClass _array_element) (Variables|Default|GetSmokeBombClass))
          (CallFunction|RemoveGUIDFromInvwithZeroCount (FortWeapon|GetInventoryGUID _array_element)))))))

(event Ability|EventFailedToActivatePassiveAbility
  (params (ActorInfo :type Gameplay%20Ability%20Actor%20Info%20Structure))
  (|CallParentFunction))

(fn CanActivateAbility
  (bind _playerpawn (Variables|Default|GetPlayerPawn))
  (bind _returnvalue (Fortnite|GetFortPlayerControllerfromActor _playerpawn))
  (Utilities|IsValid (Variables|Default|GetPlayerPawn)
    (:Is%20Valid
      (Utilities|IsValid _returnvalue
        (:Is%20Valid
          (return (<= (Inventory|CountInventoryOverflowfromAddingItem _returnvalue (Loot|CreateItemEntry "/Game/Athena/Items/Gameplay/SmokeBomb/AGID_SmokeBomb.AGID_SmokeBomb") false true) 0)))))))

(fn GiveItem
  (params (Controller :type Fort%20Player%20Controller%20Object%20Reference))
  (Utilities|FlowControl|Sequence
    (:then_0
      (bind _items (InventoryManagement|GetItemsfromInventoryOwner Controller))
      (bind _array_element (Utilities|Array|ForEachLoopwithBreak _items)
        (:LoopBody
          (bind _asfort_gadget_item_definition (Utilities|Casting|CastToFortGadgetItemDefinition (Item|GetItemDefinition _array_element)))
          (bind _returnvalue (GadgetItemDefinition|GetCharacterParts _asfort_gadget_item_definition 0))
          (bind _array_element_2 (Utilities|Array|ForEachLoopwithBreak _returnvalue)
            (:LoopBody
              (if (Utilities|Enum|Equal(Enum) (CharacterPart|Art|GetCharacterPartType _array_element_2) "Backpack")
                (Variables|Default|SetBackPartHidden false)
                (Utilities|FlowControl|Sequence
                  (:then_0
                    _)
                  (:then_1
                    _))))))))
    (:then_1
      (Inventory|GiveItemtoInventoryOwner Controller (select (Variables|Default|GetBackPartHidden) "/Game/Athena/Items/Gameplay/SmokeBomb/AGID_SmokeBomb.AGID_SmokeBomb" "/Game/Athena/Items/Gameplay/SmokeBomb/AGID_SmokeBomb_NoBack.AGID_SmokeBomb_NoBack")) ; This is the node that gives the gadget and changes character parts
    )))

(fn RemoveItem
  (params (Controller :type Fort%20Player%20Controller%20Object%20Reference))
  (Inventory|RemoveItemFromPlayer :PlayerController Controller :ItemDefinition (select (Variables|Default|GetBackPartHidden) "/Game/Athena/Items/Gameplay/SmokeBomb/AGID_SmokeBomb.AGID_SmokeBomb" "/Game/Athena/Items/Gameplay/SmokeBomb/AGID_SmokeBomb_NoBack.AGID_SmokeBomb_NoBack") :AmountToRemove -1 :bForceRemoval true :bSkipVerbRemoval false))
