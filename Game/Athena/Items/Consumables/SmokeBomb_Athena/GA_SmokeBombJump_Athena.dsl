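; Gameplay ability for the Shadow Bomb (SmokeBomb) item that gives the pawn a wall jump or double jump, outside sky tubes and vents.
; A capsule trace along movement acceleration plus an angle check selects a wall jump, launched along mirrored lateral velocity or the wall normal.
; Otherwise, a single double jump combines acceleration and scaled velocity, plays a flip or backflip, and resets on landing.
; Debug prints and TODOs remain.
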(event Ability|EventActivateAbility
  (bind _playerpawn (Variables|Default|GetPlayerPawn))
  (bind _returnvalue (Transformation|GetVelocity _playerpawn))
  (bind _returnvalue_2 (Ability|GetActivatingPawn))
  (bind _x (.x _returnvalue))
  (bind _y (.y _returnvalue))
  (bind _returnvalue_3 (Math|Vector|MakeVector _x _y))
  (bind (_asfort_player_pawn _bsuccess) (Utilities|Casting|CastToFortPlayerPawn _returnvalue_2))
  (bind _playerpawn_2 (Variables|Default|GetPlayerPawn))
  ;[1 1 1] TODO // Clamp Velocity, Fix multicast events, put trial in GC
  (Variables|Default|SetPlayerPawn _asfort_player_pawn)
  (if (Math|Boolean|ANDBoolean _bsuccess (not (SkyTube|IsinSkyTube _asfort_player_pawn :!class /Script/FortniteGame.FortPlayerPawn)) (not (GameplayTags|HasMatchingGameplayTag _asfort_player_pawn "(TagName=\"Athena.Player.InVent\")")))
    (Ability|Tasks|WaitGameplayTagAdd "(TagName=\"Athena.Player.InVent\")" 0 true
      (:then
        ;[1 1 1] Check if we're moving towards a wall, check if the angle is valid
        (bind (_impactpoint _bhit _hitresult) (CallFunction|CapsuleTrace))
        (bind _validsurface (CallFunction|CheckVerticalDotProduct _impactpoint))
        (if (and _validsurface _bhit)
          ;[1 1 1] Wall Jump
          (Development|PrintString "Wall Jump")
          (if (Math|Vector|VectorIsNearlyZero _returnvalue_3 0.01)
            (Variables|Default|SetWallJumpReflectedVector (.Normal (Collision|BreakHitResult _hitresult))) ; Lateral velocity was zero, so we're using the normal instead
            (else
              (Variables|Default|SetWallJumpReflectedVector (Math|Vector|MirrorVectorbyNormal _returnvalue_3 (.Normal (Collision|BreakHitResult _hitresult)))) ; Has lateral velocity - use it
            ))
          (Ability|ExecuteGameplayCueWithParamsOnOwner (Variables|Default|GetWallJumpCue) (Ability|GameplayCue|MakeGameplayCueParameters :NormalizedMagnitude 0.0 :RawMagnitude 0.0 :Location (.ImpactPoint (Collision|BreakHitResult _hitresult)) :Normal (.ImpactNormal (Collision|BreakHitResult _hitresult)) :GameplayEffectLevel 1 :AbilityLevel 1 :bReplicateLocationWhenUsingMinimalRepProxy false)) ; handles sounds and particles
          (Networking|ForceNetUpdate (Variables|Default|GetPlayerPawn))
          (Pawn|Character|LaunchCharacterJump (Variables|Default|GetPlayerPawn) (Math|Vector|ClampVectorSize (Math|Vector|ClampVectorSize (* (+ (Math|Vector|ClampVectorSize (* (+ (Variables|Default|GetWallJumpReflectedVector) (* (.Normal (Collision|BreakHitResult _hitresult)) (* (Math|Vector|VectorLength (Variables|Default|GetWallJumpReflectedVector)) (Variables|Default|GetWallJumpNormalMultiplier)))) (Variables|Default|GetWallJumpLateralJumpModifier)) (* (Variables|Default|GetWallJumpLateralJumpModifier) -1.0) (Variables|Default|GetWallJumpLateralJumpModifier)) (Variables|Default|GetWallJumpZOffset)) (Variables|Default|GetWallJumpForceModifier)) (* (Variables|Default|GetWallJumpForceModifier) -1.0) (Variables|Default|GetWallJumpForceModifier)) -2.7e+03 2.7e+03) false true false true)
          (Ability|Tasks|PlayMontageandWaitforTrigger "None" (Variables|Default|GetWallJumpMontage)
            (:Triggered
              (Variables|Default|SetHasDoubleJumped false)
              (Ability|EndAbility)))
          (elif (Variables|Default|GetHasDoubleJumped)
            (Ability|EndAbility)
            (else
              ;[1 1 1] Only allow one double jump until we land.  Bool re-set when landing
              (Variables|Default|SetHasDoubleJumped true)
              ;[1 1 1] Play backflip if going backwards
              (bind _jump_height (CallFunction|SetDoubleJumpForce (Variables|Default|GetPlayerPawn)))
              (Ability|Tasks|PlayMontageandWaitforTrigger "None" (select (>= (Math|Vector|DotProduct (Transformation|GetActorForwardVector (Variables|Default|GetPlayerPawn)) (Math|Vector|Normalize _jump_height)) 0.0) (Variables|Default|GetFlipMontage) (Variables|Default|GetBackFlipMontage))
                (:then
                  ;[1 1 1] Calculate double jump force and then LAUNCH
                  (Pawn|Character|LaunchCharacterJump _playerpawn_2 (Math|Vector|ClampVectorSize _jump_height -2.7e+03 2.7e+03) true true false false)
                  (Audio|SpawnSoundAttached "/Game/Athena/Sounds/Items/Shadow_Bomb/Shadow_Bomb_Double_Jump_Cue.Shadow_Bomb_Double_Jump_Cue" (Variables|/Script/Engine.Character|GetMesh (Variables|Default|GetPlayerPawn)))
                  ;[1 1 1] Ability can end in several places, but the bool to reset double jump only fires when the pawn lands
                  (EventDispatchers|AssignOnPawnLanded _playerpawn_2 (delegate Custom|Landed)))
                (:Triggered
                  _)
                (:Cancelled
                  _))
              (Ability|EndAbility)))))
      (:Added
        _)))
  (Ability|EndAbility))

(event Custom|Landed
  (params (Hit :type Hit%20Result%20Structure%20%28by%20ref%29))
  ;[1 1 1] Ability can end in several places, but the bool to reset double jump only fires when the pawn lands
  (Variables|Default|SetHasDoubleJumped false))

(fn CalculateWallJumpForce
  (params (HitResult :type Hit%20Result%20Structure))
  (bind (_bblockinghit _binitialoverlap _time _distance _location _impactpoint _normal _impactnormal) (Collision|BreakHitResult HitResult))
  (return (Math|Vector|ClampVectorSize (* (+ (Math|Vector|ClampVectorSize (* (+ (Variables|Default|GetWallJumpReflectedVector) (* _normal (* (Math|Vector|VectorLength (Variables|Default|GetWallJumpReflectedVector)) (Variables|Default|GetWallJumpNormalMultiplier)))) (Variables|Default|GetWallJumpLateralJumpModifier)) (* (Variables|Default|GetWallJumpLateralJumpModifier) -1.0) (Variables|Default|GetWallJumpLateralJumpModifier)) (Variables|Default|GetWallJumpZOffset)) (Variables|Default|GetWallJumpForceModifier)) (* (Variables|Default|GetWallJumpForceModifier) -1.0) (Variables|Default|GetWallJumpForceModifier)) _normal _impactpoint _impactnormal))

(fn CapsuleTrace
  (bind (_outhit _returnvalue) (Collision|CapsuleTraceByChannel (Transformation|GetActorLocation (Variables|Default|GetPlayerPawn)) (+ (Transformation|GetActorLocation (Variables|Default|GetPlayerPawn)) (* (Math|Vector|Normalize (Pawn|Components|CharacterMovement|GetCurrentAcceleration (Variables|/Script/Engine.Character|GetCharacterMovement (Variables|Default|GetPlayerPawn)))) 1e+02)) (* (Variables|/Script/Engine.CapsuleComponent|GetCapsuleRadius (Variables|/Script/Engine.Character|GetCapsuleComponent (Variables|Default|GetPlayerPawn))) 1.75) (* (Variables|/Script/Engine.CapsuleComponent|GetCapsuleHalfHeight (Variables|/Script/Engine.Character|GetCapsuleComponent (Variables|Default|GetPlayerPawn))) 1.25) "TraceTypeQuery10"))
  (bind _asbp_athena_sky_tube (Utilities|Casting|CastToBP_Athena_SkyTube (.HitActor (Collision|BreakHitResult _outhit)))
    (:then
      (if (not (SkyTube|IsSkyTubeEnabled _asbp_athena_sky_tube))
        (return (.ImpactPoint (Collision|BreakHitResult _outhit)) false _outhit)))
    (:CastFailed
      _))
  (return (.ImpactPoint (Collision|BreakHitResult _outhit)) _returnvalue _outhit))

(fn CheckVerticalDotProduct
  (params (WallJumpImpactPoint :type Vector))
  (bind _playerpawn (Variables|Default|GetPlayerPawn))
  (bind _returnvalue (Transformation|GetActorUpVector _playerpawn))
  (bind _playerpawn_2 (Variables|Default|GetPlayerPawn))
  (bind _returnvalue_2 (Transformation|GetActorLocation _playerpawn_2))
  (bind _returnvalue_3 (- WallJumpImpactPoint _returnvalue_2))
  (bind _returnvalue_4 (Math|Vector|Normalize _returnvalue_3))
  (bind _returnvalue_5 (Math|Vector|Normalize _returnvalue))
  (bind _returnvalue_6 (Math|Vector|DotProduct _returnvalue_4 _returnvalue_5))
  (Development|PrintString (Utilities|String|ToString(Float) _returnvalue_6) true true "(R=1.000000,G=0.532413,B=0.000000,A=1.000000)" 3.0)
  (return (Math|Float|InRange(Float) _returnvalue_6 (* (Variables|Default|GetWallJumpAngleThreshold) -1.0) (Variables|Default|GetWallJumpAngleThreshold) false false)))

(fn OnRep_AdditionalJumpZ)

(fn OnRep_JumpHeight)

(fn SetDoubleJumpForce
  (params (PlayerPawn :type Fort%20Player%20Pawn%20Object%20Reference))
  (bind (_returnvalue_x _returnvalue_y _returnvalue_z) (Math|Vector|BreakVector (* (Variables|/Script/Engine.MovementComponent|GetVelocity (Variables|/Script/Engine.Character|GetCharacterMovement PlayerPawn)) (Variables|Default|GetVerticalVelocityMulitplier))))
  (bind (_returnvalue_x_2 _returnvalue_y_2) (Math|Vector|BreakVector (* (Variables|/Script/Engine.MovementComponent|GetVelocity (Variables|/Script/Engine.Character|GetCharacterMovement PlayerPawn)) (Variables|Default|GetHorizontalVelocityMultiplier))))
  (return (Math|Vector|vector+vector (Math|Vector|vector+vector (Math|Vector|vector*vector (Math|Vector|Normalize (Pawn|Components|CharacterMovement|GetCurrentAcceleration (Variables|/Script/Engine.Character|GetCharacterMovement PlayerPawn))) (Math|Vector|MakeVector (Variables|Default|GetHorizontalMultiplier) (Variables|Default|GetHorizontalMultiplier))) (Math|Vector|MakeVector 0.0 0.0 (Variables|Default|GetVerticalMultiplier))) (Math|Vector|MakeVector _returnvalue_x_2 _returnvalue_y_2 _returnvalue_z))))
