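; Passive gameplay ability for the Athena Smoke Bomb that cleans up the item when it expires or its owner dies.
; On activation it caches the player pawn, waits for the Abilities.AthenaSmokeBomb.Expired event and binds the pawn's OnDied.
; Both paths duplicate one check: force-remove any remaining smoke bombs, or, if none remain, unblock equipment change as a failsafe.
; The death handler is noted as a fix keeping the server authoritative on death.
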(event Ability|EventActivateAbility
  (bind _asfort_player_pawn (Utilities|Casting|CastToFortPlayerPawn (Ability|GetAvatarActorfromActorInfo)))
  (Variables|Default|SetPlayerPawn _asfort_player_pawn)
  (Ability|Tasks|WaitGameplayEvent "(TagName=\"Abilities.AthenaSmokeBomb.Expired\")"
    (:then
      (EventDispatchers|AssignOnDied _asfort_player_pawn (delegate Custom|InstigatorDied)))
    (:EventReceived
      (bind _asfort_player_controller (Utilities|Casting|CastToFortPlayerController (Pawn|GetController (Variables|Default|GetPlayerPawn))))
      (if (== (Inventory|GetItemQuantityOnPlayer _asfort_player_controller (Variables|Default|GetSmokeBombItemDef)) 0)
        (Tags|UnBlockAbilitiesWithTags (Utilities|Casting|CastToFortAbilitySystemComponent (Ability|GetAbilitySystemComponent (Variables|Default|GetPlayerPawn))) (Variables|Default|GetChangeEquipmentTag) false) ; Failsafe for unblocking equipment change
        (Inventory|RemoveItemFromPlayer :PlayerController _asfort_player_controller :ItemDefinition (Variables|Default|GetSmokeBombItemDef) :AmountToRemove -1 :bForceRemoval true :bSkipVerbRemoval false)))))

(event Custom|InstigatorDied
  (params (DamagedActor :type Actor%20Object%20Reference) (Damage :type Float%20%28single-precision%29) (InstigatedBy :type Controller%20Object%20Reference) (DamageCauser :type Actor%20Object%20Reference) (HitLocation :type Vector) (FHitComponent :type Primitive%20Component%20Object%20Reference) (BoneName :type Name) (Momentum :type Vector))
  ;[1 1 1] This makes sure the server is authoritative on death.  Fix for FORT-174239
  (bind _asfort_player_pawn (Utilities|Casting|CastToFortPlayerPawn DamagedActor))
  (Variables|Default|SetPlayerPawn _asfort_player_pawn)
  (bind _asfort_player_controller (Utilities|Casting|CastToFortPlayerController (Pawn|GetController (Variables|Default|GetPlayerPawn))))
  (if (== (Inventory|GetItemQuantityOnPlayer _asfort_player_controller (Variables|Default|GetSmokeBombItemDef)) 0)
    (Tags|UnBlockAbilitiesWithTags (Utilities|Casting|CastToFortAbilitySystemComponent (Ability|GetAbilitySystemComponent (Variables|Default|GetPlayerPawn))) (Variables|Default|GetChangeEquipmentTag) false) ; Failsafe for unblocking equipment change
    (Inventory|RemoveItemFromPlayer :PlayerController _asfort_player_controller :ItemDefinition (Variables|Default|GetSmokeBombItemDef) :AmountToRemove -1 :bForceRemoval true :bSkipVerbRemoval false)))
